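# Translucent UI material for Crew benefit tile cards, drawing a themed gradient background with rounded corners.
# An inverted, cubic-eased vertical UV gradient serves as the base for blending the ColorTop and ColorBot parameters.
# A Color Dodge layer at reduced strength finishes the color, alongside a screen-space random noise source that likely adds grain.
# A CoreUI rounded mask supplies alpha, with per-corner radii and a pixel-units toggle.
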Material /Crew/UI/Materials/CrewBenefits/Tile/M_UI_Benefit_Tile_Card_Material
graph {
 properties {bAutomaticallySetUsageInEditor=false, blendMode=BLEND_Translucent, materialDomain=MD_UI, usedShadingModels=Default Lit}
 nodes [
  n1 = ComponentMask(None=TextureCoordinate()) {g=true}
  NamedRerouteDeclaration(None=MaterialFunctionCall!("Value (0-1)"=OneMinus(None=@n1)) {materialFunction=/FortUILibrary/UGC/MaterialFunctions/MF_UGCUI_EaseCurves}.EaseInCubic) {name=Base Gradient, nodeColor=<rgba 0 0 0 1>}
  n2 = MaterialFunctionCall(
   R=ScalarParameter() {defaultValue=16, group=Layout, parameterName=TopLeftCornerRadius, sortPriority=4}
   G=ScalarParameter() {defaultValue=16, group=Layout, parameterName=BottomLeftCornerRadius, sortPriority=7}
   B=ScalarParameter() {defaultValue=16, group=Layout, parameterName=TopRightCornerRadius, sortPriority=5}
   A=ScalarParameter() {defaultValue=16, group=Layout, parameterName=BottomRightCornerRadius, sortPriority=6}
  ) {materialFunction=&AppendMany}
  n3 = ScalarParameter() {
   defaultValue=1
   desc="If enabled reads corner radius values in pixels.\r\nbool"
   group=Layout
   parameterName=UseValueInPixels
   sliderMax=1
   sortPriority=1
  }
  NamedRerouteDeclaration(None=NamedRerouteDeclaration(None=MaterialFunctionCall(Corner Radius=@n2.RGBA, Radius in Pixels=@n3) {materialFunction=&MF_UI_CoreUI_RoundedMask}.Shape Mask) {name=Shape Mask, nodeColor=<rgba 0 0 0 1>}) {name=Final Alpha, nodeColor=<rgba 1 0.5391 0 1>}
  n4 = NamedRerouteUsage()
  n5 = LinearInterpolate(A=NamedRerouteUsage(), B=NamedRerouteUsage(), Alpha=Multiply(A=NamedRerouteUsage()))
  NamedRerouteDeclaration(None=MaterialFunctionCall!(Blend=Multiply(A=@n4) {constB=0.2}, Base=NamedRerouteDeclaration(None=@n5) {name=Colored Gradient, nodeColor=<rgba 0 0 0 1>}) {materialFunction=/Engine/Functions/Engine_MaterialFunctions03/Blends/Blend_ColorDodge}.Result) {name=FInal Color, nodeColor=<rgba 0.2188 0 1 1>}
  NamedRerouteDeclaration(None=VectorParameter() {group=Theming, parameterName=ColorTop, defaultValue=<rgba 0.0953 0.4397 0.6445 1>}.RGB) {name=Color 01, nodeColor=<rgba 0 0 0 1>}
  NamedRerouteDeclaration(None=VectorParameter() {group=Theming, parameterName=ColorBot, defaultValue=<rgba 0.0086 0.0273 0.1195 1>}.RGB) {name=Color 02, nodeColor=<rgba 0 0 0 1>}
  NamedRerouteDeclaration(None=MaterialFunctionCall!(UV=ScreenPosition().ViewportUV, Seed) {materialFunction=&MF_UI_Random}.Result) {name=NoiseTextureColor, nodeColor=<rgba 0 0.3906 1 1>}
  NamedRerouteUsage()
  NamedRerouteUsage()
 ]
}
