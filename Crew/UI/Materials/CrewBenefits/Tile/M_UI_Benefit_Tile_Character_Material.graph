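# Translucent UI material for a Crew benefit tile, built to show a character texture with soft fades inside a rounded card.
# UI UVs are offset, ratio-corrected and center-scaled to sample MainTexture, with parameters for placement and scale.
# Eased top and bottom gradients, a stretched radial gradient and a CoreUI rounded-corner mask form named masks.
# Named reroutes organize color and alpha, with final alpha a saturated product of five rerouted values.
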Material /Crew/UI/Materials/CrewBenefits/Tile/M_UI_Benefit_Tile_Character_Material
graph {
 expressions {
  MaterialExpressionAdd_0 {
   class=MaterialExpressionAdd
   inputs [
    A {from=MaterialExpressionConstant_3}
    B {from=MaterialExpressionLinearInterpolate_0}
   ]
   properties {constB=0.2}
  }
  MaterialExpressionAdd_1 {
   class=MaterialExpressionAdd
   inputs [
    A {from=MaterialExpressionNamedRerouteDeclaration_10}
    B {from=MaterialExpressionConstant2Vector_0}
   ]
  }
  MaterialExpressionAdd_10 {
   class=MaterialExpressionAdd
   inputs [
    A {from=MaterialExpressionNamedRerouteUsage_12}
    B {from=MaterialExpressionConstant2Vector_6}
   ]
  }
  MaterialExpressionAdd_2 {
   class=MaterialExpressionAdd
   inputs [
    A {from=MaterialExpressionMaterialFunctionCall_6, output=ScaleY}
    B {from=MaterialExpressionConstant2Vector_2}
   ]
  }
  MaterialExpressionAdd_6 {
   class=MaterialExpressionAdd
   inputs [
    A {from=MaterialExpressionScalarParameter_4}
    B {from=MaterialExpressionLinearInterpolate_4}
   ]
   properties {constB=0.2}
  }
  MaterialExpressionAdd_9 {
   class=MaterialExpressionAdd
   inputs [
    A {from=MaterialExpressionAdd_1}
    B {from=MaterialExpressionAppendVector_1}
   ]
   properties {constB=0.5}
  }
  MaterialExpressionAppendVector_1 {
   class=MaterialExpressionAppendVector
   inputs [
    A {from=MaterialExpressionScalarParameter_5, output=""}
    B {from=MaterialExpressionScalarParameter_0, output=""}
   ]
  }
  MaterialExpressionComponentMask_1 {
   class=MaterialExpressionComponentMask
   inputs [None {from=MaterialExpressionMaterialFunctionCall_14, output=Normalized UV}]
   properties {g=true}
  }
  MaterialExpressionComponentMask_3 {
   class=MaterialExpressionComponentMask
   inputs [None {from=MaterialExpressionMaterialFunctionCall_21, output=Normalized UV}]
   properties {g=true}
  }
  MaterialExpressionConstant2Vector_0 {class=MaterialExpressionConstant2Vector, properties {g=0.5}}
  MaterialExpressionConstant2Vector_2 {class=MaterialExpressionConstant2Vector, properties {g=-0.5}}
  MaterialExpressionConstant2Vector_5 {class=MaterialExpressionConstant2Vector, properties {g=2, r=1.25}}
  MaterialExpressionConstant2Vector_6 {class=MaterialExpressionConstant2Vector, properties {g=0.3}}
  MaterialExpressionConstant_20 {class=MaterialExpressionConstant, properties {r=0.1}}
  MaterialExpressionConstant_21 {class=MaterialExpressionConstant, properties {r=0.4}}
  MaterialExpressionConstant_3 {class=MaterialExpressionConstant, properties {r=0.9}}
  MaterialExpressionConstant_7 {class=MaterialExpressionConstant, properties {r=1}}
  MaterialExpressionConvert_1 {
   class=MaterialExpressionConvert
   inputs [
    None {from=MaterialExpressionScalarParameter_7}
    None {from=MaterialExpressionScalarParameter_8}
   ]
   properties {
    nodeName=Make Vector2
    convertInputs [
     {
      type=Scalar
      expressionInput {
       expression=&self:MaterialExpressionScalarParameter_7
       inputName=None
       mask=0
       maskA=0
       maskB=0
       maskG=0
       maskR=0
       outputIndex=0
      }
     }
     {
      type=Scalar
      expressionInput {
       expression=&self:MaterialExpressionScalarParameter_8
       inputName=None
       mask=0
       maskA=0
       maskB=0
       maskG=0
       maskR=0
       outputIndex=0
      }
     }
    ]
    convertOutputs [{type=Vector2}]
   }
  }
  MaterialExpressionInverseLinearInterpolate_0 {
   class=MaterialExpressionInverseLinearInterpolate
   inputs [
    A {from=MaterialExpressionConstant_3, output=""}
    B {from=MaterialExpressionAdd_0, output=""}
    Value {from=MaterialExpressionOneMinus_3, output=""}
   ]
   properties {}
  }
  MaterialExpressionInverseLinearInterpolate_2 {
   class=MaterialExpressionInverseLinearInterpolate
   inputs [
    A {from=MaterialExpressionScalarParameter_4, output=""}
    B {from=MaterialExpressionAdd_6, output=""}
    Value {from=MaterialExpressionComponentMask_3, output=""}
   ]
   properties {}
  }
  MaterialExpressionLinearInterpolate_0 {
   class=MaterialExpressionLinearInterpolate
   inputs [
    A {from=""}
    B {from=MaterialExpressionSubtract_0}
    Alpha {from=MaterialExpressionConstant_7}
   ]
   properties {constA=0.01}
  }
  MaterialExpressionLinearInterpolate_4 {
   class=MaterialExpressionLinearInterpolate
   inputs [
    A {from=""}
    B {from=MaterialExpressionSubtract_2}
    Alpha {from=MaterialExpressionScalarParameter_6}
   ]
   properties {constA=0.01}
  }
  MaterialExpressionMaterialFunctionCall_11 {
   class=MaterialExpressionMaterialFunctionCall
   inputs [
    UV {from=MaterialExpressionMaterialFunctionCall_15, output=UVs}
    GradientStart {from=MaterialExpressionConstant_20}
    GradientEnd {from=MaterialExpressionConstant_21}
   ]
   properties {materialFunction=&MF_UI_RadialGradient}
  }
  MaterialExpressionMaterialFunctionCall_12 {
   class=MaterialExpressionMaterialFunctionCall
   inputs ["Value (0-1)" {from=MaterialExpressionSaturate_1}]
   properties {materialFunction=/FortUILibrary/UGC/MaterialFunctions/MF_UGCUI_EaseCurves}
  }
  MaterialExpressionMaterialFunctionCall_13 {
   class=MaterialExpressionMaterialFunctionCall
   inputs [
    Number To Convert {from=MaterialExpressionLinearInterpolate_0}
    Number of Bits {from=""}
    UVs {from=""}
   ]
   properties {materialFunction=/Engine/Functions/Engine_MaterialFunctions01/Debug/DebugBinaryValues-Float}
  }
  MaterialExpressionMaterialFunctionCall_14 {class=MaterialExpressionMaterialFunctionCall, inputs [], properties {materialFunction=&GetUserInterfaceUV}}
  MaterialExpressionMaterialFunctionCall_15 {
   class=MaterialExpressionMaterialFunctionCall
   inputs [
    UVs {from=MaterialExpressionAdd_10}
    Texture Scale {from=MaterialExpressionConstant2Vector_5}
   ]
   properties {materialFunction=&ScaleUVsByCenter}
  }
  MaterialExpressionMaterialFunctionCall_17 {
   class=MaterialExpressionMaterialFunctionCall
   inputs [
    UVs {from=MaterialExpressionAdd_2}
    Texture Scale {from=MaterialExpressionScalarParameter_1}
   ]
   properties {materialFunction=&ScaleUVsByCenter}
  }
  MaterialExpressionMaterialFunctionCall_18 {
   class=MaterialExpressionMaterialFunctionCall
   inputs [
    UV {from=""}
    Size {from=""}
    Corner Radius {from=MaterialExpressionScalarParameter_3}
    Radius in Pixels {from=MaterialExpressionScalarParameter_2}
    Skew Angle {from=""}
    Outline Thickness {from=""}
    Corners Modulated by Size {from=""}
    Pixel Size {from=""}
   ]
   properties {materialFunction=&MF_UI_CoreUI_RoundedMask}
  }
  MaterialExpressionMaterialFunctionCall_19 {
   class=MaterialExpressionMaterialFunctionCall
   inputs ["Value (0-1)" {from=MaterialExpressionSaturate_3}]
   properties {materialFunction=&MF_UGCUI_EaseCurves}
  }
  MaterialExpressionMaterialFunctionCall_20 {
   class=MaterialExpressionMaterialFunctionCall
   inputs [
    Number To Convert {from=MaterialExpressionLinearInterpolate_4}
    Number of Bits {from=""}
    UVs {from=""}
   ]
   properties {materialFunction=&DebugBinaryValues-Float}
  }
  MaterialExpressionMaterialFunctionCall_21 {class=MaterialExpressionMaterialFunctionCall, inputs [], properties {materialFunction=&GetUserInterfaceUV}}
  MaterialExpressionMaterialFunctionCall_4 {class=MaterialExpressionMaterialFunctionCall, inputs [], properties {materialFunction=&GetUserInterfaceUV}}
  MaterialExpressionMaterialFunctionCall_6 {
   class=MaterialExpressionMaterialFunctionCall
   inputs [
    UV {from=MaterialExpressionAdd_9}
    TextureSize {from=MaterialExpressionConvert_1}
    ScalePivotPoint {from=""}
    ContainerSize {from=""}
   ]
   properties {materialFunction=&MF_UI_UVScaling}
  }
  MaterialExpressionMaterialFunctionCall_9 {
   class=MaterialExpressionMaterialFunctionCall
   inputs ["Value (0-1)" {from=MaterialExpressionMaterialFunctionCall_11, output=Linear}]
   properties {materialFunction=&MF_UI_EaseCurves}
  }
  MaterialExpressionMultiply_1 {
   class=MaterialExpressionMultiply
   inputs [
    A {from=MaterialExpressionMultiply_3}
    B {from=MaterialExpressionNamedRerouteUsage_10}
   ]
  }
  MaterialExpressionMultiply_2 {
   class=MaterialExpressionMultiply
   inputs [
    A {from=MaterialExpressionMultiply_1}
    B {from=MaterialExpressionNamedRerouteUsage_8}
   ]
  }
  MaterialExpressionMultiply_3 {
   class=MaterialExpressionMultiply
   inputs [
    A {from=MaterialExpressionNamedRerouteUsage_5}
    B {from=MaterialExpressionNamedRerouteUsage_13}
   ]
  }
  MaterialExpressionMultiply_4 {
   class=MaterialExpressionMultiply
   inputs [
    A {from=MaterialExpressionMultiply_2}
    B {from=MaterialExpressionNamedRerouteUsage_14}
   ]
  }
  MaterialExpressionMultiply_6 {
   class=MaterialExpressionMultiply
   inputs [
    A {from=MaterialExpressionOneMinus_6}
    B {from=""}
   ]
   properties {constB=0.5}
  }
  MaterialExpressionNamedRerouteDeclaration_0 {
   class=MaterialExpressionNamedRerouteDeclaration
   inputs [None {from=MaterialExpressionTextureSampleParameter2D_0, output=RGB}]
   properties {name=TextureColor, nodeColor=<rgba 0 0.3906 1 1>}
  }
  MaterialExpressionNamedRerouteDeclaration_1 {
   class=MaterialExpressionNamedRerouteDeclaration
   inputs [None {from=MaterialExpressionTextureSampleParameter2D_0, output=A}]
   properties {name=TextureAlpha, nodeColor=<rgba 0 0.3906 1 1>}
  }
  MaterialExpressionNamedRerouteDeclaration_10 {
   class=MaterialExpressionNamedRerouteDeclaration
   inputs [None {from=MaterialExpressionMaterialFunctionCall_4, output=Normalized UV}]
   properties {name=NormalizedUV, nodeColor=<rgba 0.1953 0 1 1>}
  }
  MaterialExpressionNamedRerouteDeclaration_11 {
   class=MaterialExpressionNamedRerouteDeclaration
   inputs [None {from=MaterialExpressionMaterialFunctionCall_18, output=Shape Mask}]
   properties {name=CardMask, nodeColor=<rgba 1 0.9844 0 1>}
  }
  MaterialExpressionNamedRerouteDeclaration_3 {
   class=MaterialExpressionNamedRerouteDeclaration
   inputs [None {from=MaterialExpressionMaterialFunctionCall_12, output=EaseInOutQuint}]
   properties {name=TopFadeMask, nodeColor=<rgba 0.6875 1 0 1>}
  }
  MaterialExpressionNamedRerouteDeclaration_5 {
   class=MaterialExpressionNamedRerouteDeclaration
   inputs [None {from=MaterialExpressionMaterialFunctionCall_19, output=EaseInCubic}]
   properties {name=BottomFadeMask, nodeColor=<rgba 0.6875 1 0 1>}
  }
  MaterialExpressionNamedRerouteDeclaration_7 {
   class=MaterialExpressionNamedRerouteDeclaration
   inputs [None {from=MaterialExpressionNamedRerouteUsage_4}]
   properties {name=FinalColor, nodeColor=<rgba 1 0 0.9823 1>}
  }
  MaterialExpressionNamedRerouteDeclaration_8 {
   class=MaterialExpressionNamedRerouteDeclaration
   inputs [None {from=MaterialExpressionSaturate_5}]
   properties {name=FinalAlpha, nodeColor=<rgba 1 0 0.9823 1>}
  }
  MaterialExpressionNamedRerouteDeclaration_9 {
   class=MaterialExpressionNamedRerouteDeclaration
   inputs [None {from=MaterialExpressionOneMinus_7}]
   properties {name=RadialMask, nodeColor=<rgba 0 0 0 1>}
  }
  MaterialExpressionNamedRerouteUsage_10 {class=MaterialExpressionNamedRerouteUsage}
  MaterialExpressionNamedRerouteUsage_12 {class=MaterialExpressionNamedRerouteUsage}
  MaterialExpressionNamedRerouteUsage_13 {class=MaterialExpressionNamedRerouteUsage}
  MaterialExpressionNamedRerouteUsage_14 {class=MaterialExpressionNamedRerouteUsage}
  MaterialExpressionNamedRerouteUsage_4 {class=MaterialExpressionNamedRerouteUsage}
  MaterialExpressionNamedRerouteUsage_5 {class=MaterialExpressionNamedRerouteUsage}
  MaterialExpressionNamedRerouteUsage_6 {class=MaterialExpressionNamedRerouteUsage}
  MaterialExpressionNamedRerouteUsage_7 {class=MaterialExpressionNamedRerouteUsage}
  MaterialExpressionNamedRerouteUsage_8 {class=MaterialExpressionNamedRerouteUsage}
  MaterialExpressionOneMinus_3 {class=MaterialExpressionOneMinus, inputs [None {from=MaterialExpressionComponentMask_1}]}
  MaterialExpressionOneMinus_4 {class=MaterialExpressionOneMinus, inputs [None {from=MaterialExpressionInverseLinearInterpolate_0}]}
  MaterialExpressionOneMinus_5 {class=MaterialExpressionOneMinus, inputs [None {from=MaterialExpressionInverseLinearInterpolate_2}]}
  MaterialExpressionOneMinus_6 {class=MaterialExpressionOneMinus, inputs [None {from=MaterialExpressionMaterialFunctionCall_9, output=EaseInOutSine}]}
  MaterialExpressionOneMinus_7 {class=MaterialExpressionOneMinus, inputs [None {from=MaterialExpressionMultiply_6}]}
  MaterialExpressionSaturate_1 {class=MaterialExpressionSaturate, inputs [None {from=MaterialExpressionOneMinus_4}]}
  MaterialExpressionSaturate_3 {class=MaterialExpressionSaturate, inputs [None {from=MaterialExpressionOneMinus_5}]}
  MaterialExpressionSaturate_5 {class=MaterialExpressionSaturate, inputs [None {from=MaterialExpressionMultiply_4}]}
  MaterialExpressionScalarParameter_0 {class=MaterialExpressionScalarParameter, properties {group=Texture, parameterName=TextureVerticalOffset}}
  MaterialExpressionScalarParameter_1 {class=MaterialExpressionScalarParameter, properties {defaultValue=3, group=Texture, parameterName=TextureScale}}
  MaterialExpressionScalarParameter_2 {class=MaterialExpressionScalarParameter, properties {defaultValue=1, group=CardMask, parameterName=RadiusInPixel, sliderMax=1}}
  MaterialExpressionScalarParameter_3 {class=MaterialExpressionScalarParameter, properties {defaultValue=20, group=CardMask, parameterName=CornerRadius}}
  MaterialExpressionScalarParameter_4 {class=MaterialExpressionScalarParameter, properties {defaultValue=0.6, parameterName=BottomFadeSize, sliderMax=1}}
  MaterialExpressionScalarParameter_5 {class=MaterialExpressionScalarParameter, properties {parameterName=TextureHorizontalOffset}}
  MaterialExpressionScalarParameter_6 {class=MaterialExpressionScalarParameter, properties {defaultValue=12, parameterName=BottomFadeStretch}}
  MaterialExpressionScalarParameter_7 {class=MaterialExpressionScalarParameter, properties {defaultValue=275, parameterName=RatioX}}
  MaterialExpressionScalarParameter_8 {class=MaterialExpressionScalarParameter, properties {defaultValue=374, parameterName=RatioY}}
  MaterialExpressionSubtract_0 {
   class=MaterialExpressionSubtract
   inputs [
    A {from=""}
    B {from=MaterialExpressionConstant_3}
   ]
  }
  MaterialExpressionSubtract_2 {
   class=MaterialExpressionSubtract
   inputs [
    A {from=""}
    B {from=MaterialExpressionScalarParameter_4}
   ]
  }
  MaterialExpressionTextureSampleParameter2D_0 {
   class=MaterialExpressionTextureSampleParameter2D
   inputs [
    UVs {from=MaterialExpressionMaterialFunctionCall_17, output=UVs}
    Apply View MipBias {from="", output=""}
   ]
   properties {parameterName=MainTexture, samplerSource=SSM_Clamp_WorldGroupSettings, texture=/Crew/UI/Textures/WIP_Test_Texture/T_Crew_Styles}
  }
 }
 properties {bAutomaticallySetUsageInEditor=false, blendMode=BLEND_Translucent, materialDomain=MD_UI, usedShadingModels=Default Lit}
}
